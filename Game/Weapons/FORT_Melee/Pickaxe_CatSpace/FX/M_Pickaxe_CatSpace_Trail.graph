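# Translucent unlit, two-sided Niagara ribbon material for the CatSpace pickaxe trail, with a tiled-square look that dissolves along its length.
# Color combines a start-to-end gradient, hue-shifted square tiles scrolling in quantized time steps, camera-space pulsing bright bands and edge outlines, tinted by particle color.
# A panning square SDF, offset by a dynamic parameter, cuts the dissolve and its glowing stroke.
# Opacity adds soft edges and noise-distorted end fading.
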Material /Game/Weapons/FORT_Melee/Pickaxe_CatSpace/FX/M_Pickaxe_CatSpace_Trail
graph {
 properties {
  bUsedWithNiagaraRibbons=true
  blendMode=BLEND_Translucent
  pixelDepthOffsetMode=PDOM_Legacy
  shadingModel=MSM_Unlit
  twoSided=true
  usedShadingModels=Unlit
 }
 nodes [
  n1 = ScalarParameter() {defaultValue=1, group=Square, parameterName=Square Tiling X, sliderMax=5}
  n2 = ScalarParameter() {defaultValue=1.5, group=Square, parameterName=Square Tiling Y, sliderMax=5}
  n3 = Time()
  n4 = ScalarParameter() {defaultValue=0.7, parameterName=Panning Speed, sliderMax=2, sliderMin=-2}
  n5 = Add(
   A=MaterialFunctionCall(UVs=TextureCoordinate() {coordinateIndex=1}, Texture Scale=AppendVector(A=@n1, B=@n2)) {materialFunction=&ScaleUVsByCenter}.UVs
   B=AppendVector(A=Multiply(A=@n3, B=@n4) {constB=0.7}, B=Constant())
  )
  n6 = ComponentMask(None=TextureCoordinate()) {r=true}
  n7 = DynamicParameter() {defaultValue=<rgba 0.25 0.25 1 1>, paramNames [Dissolve, Dissolve square, Param3, Param4]}
  n8 = Saturate(None=Multiply(A=Add(A=@n6, B=@n7.Dissolve square), B=ScalarParameter() {defaultValue=1, parameterName=Dissolve Range, sliderMax=5}))
  n9 = Divide(
   A=Subtract(
    A=TextureSampleParameter2D(UVs=@n5) {parameterName=Dissolve Texture, texture=&/T_SquareSDF_02}.R
    B=ConstantBiasScale(None=@n8) {bias=-0.5, scale=2}
   )
   B=ScalarParameter() {defaultValue=0.25, parameterName=Max}
  )
  n10 = OneMinus(None=Subtract(A=Abs(None=@n9), B=ScalarParameter() {defaultValue=1, parameterName=Stroke, sliderMax=5}))
  MaterialFunctionCall(Emissive=LinearInterpolate(
    A=NamedRerouteUsage()
    B=VectorParameter() {parameterName=Color Fade Outline, defaultValue=<rgba 2 0 1.6813 1>}.RGB
    Alpha=Saturate(None=@n10)
   )) {materialFunction=&MF_FX_EmissiveAdjustments}
  n11 = ComponentMask(None=TextureCoordinate()) {r=true}
  Multiply(A=Multiply(A=OneMinus(None=@n11), B=Saturate(None=@n9)), B=NamedRerouteUsage())
  n12 = ComponentMask(None=TextureCoordinate() {coordinateIndex=1}) {g=true, r=true}
  n13 = ScalarParameter() {defaultValue=0.3, parameterName=Square Tiling 02, sliderMax=10}
  n14 = Time()
  n15 = ScalarParameter() {defaultValue=16, parameterName=Step}
  n16 = Divide(A=Floor(None=Multiply(A=@n14, B=@n15) {constB=10}), B=@n15) {constB=10}
  n17 = Multiply(A=@n16, B=ScalarParameter() {parameterName=Step Speed Big}) {constB=0}
  n18 = Constant()
  n19 = TextureSampleParameter2D(UVs=Add(A=Multiply(A=@n12, B=@n13) {constB=0.5}, B=AppendVector(A=@n17, B=@n18))) {parameterName=Color Texture, samplerType=SAMPLERTYPE_LinearGrayscale, texture=&/T_SquareTile}
  n20 = ScalarParameter() {defaultValue=1, parameterName=Square Tiling, sliderMax=10}
  n21 = Multiply(A=@n16, B=ScalarParameter() {parameterName=Step Speed}) {constB=3}
  n22 = Constant()
  n23 = TextureSampleParameter2D(UVs=Add(A=Multiply(A=@n12, B=@n20), B=AppendVector(A=@n21, B=@n22))) {parameterName=Color Texture, samplerType=SAMPLERTYPE_LinearGrayscale, texture=&T_SquareTile}
  NamedRerouteDeclaration(None=Min(A=@n19.R, B=@n23.R)) {name=Square Tiles, nodeColor=<rgba 0.875 1 0 1>}
  n24 = NamedRerouteUsage()
  n25 = OneMinus(None=@n24)
  n26 = ScalarParameter() {defaultValue=-0.05, parameterName=Hue Shift, sliderMax=1, sliderMin=-1}
  n27 = VectorParameter() {parameterName=Color End, defaultValue=<rgba 1 0 0.7297 1>}
  n28 = VectorParameter() {parameterName=Color Start, defaultValue=<rgba 0.8719 0.0625 1 1>}
  n29 = TextureCoordinate()
  n30 = Power(Base=OneMinus(None=ComponentMask(None=@n29) {r=true})) {constExponent=5}
  n31 = Multiply(
   A=Saturate(None=Add(A=@n24) {constB=0.6})
   B=MaterialFunctionCall(Hue Shift Percentage=Multiply(A=@n25, B=@n26) {constB=-0.05}, Texture=LinearInterpolate(A=@n27.RGB, B=@n28.RGB, Alpha=@n30)) {materialFunction=&HueShift}.Result
  )
  n32 = OneMinus(None=ComponentMask(None=TextureCoordinate()) {r=true})
  n33 = WorldPosition()
  n34 = Multiply(
   A=ComponentMask(None=TransformPosition(None=@n33.XYZ) {transformSourceType=TRANSFORMPOSSOURCE_World, transformType=TRANSFORMPOSSOURCE_Camera}) {g=true}
   B=ScalarParameter() {defaultValue=0.1, parameterName=Emissive Lines Tiling, sliderMax=1}
  ) {constB=0.1}
  n35 = Multiply(A=Time()) {constB=2}
  n36 = Subtract(A=Sine(None=Add(A=@n34, B=@n35))) {constB=0.75}
  n37 = Constant() {r=1}
  n38 = Constant()
  n39 = LinearInterpolate(A=@n31, B=Multiply(A=@n31) {constB=2}, Alpha=Multiply(A=Power(Base=@n32), B=If(A=@n36, "A > B"=@n37, "A < B"=@n38)))
  n40 = Constant() {r=0.5}
  n41 = ScalarParameter() {defaultValue=0.05, parameterName=Outline Thickness, sliderMax=0.1}
  n42 = OneMinus(None=SphereMask(A=ComponentMask(None=@n29) {g=true}, B=@n40, Radius=Subtract(A=@n40, B=@n41)) {attenuationRadius=0.46, hardnessPercent=95})
  n43 = LinearInterpolate(A=@n27.RGB, B=VectorParameter() {parameterName=Color Outline, defaultValue=<rgba 0.8719 0.0625 1 1>}.RGB, Alpha=@n42)
  n44 = ParticleColor()
  NamedRerouteDeclaration(None=Multiply(A=LinearInterpolate(A=@n39, B=@n43, Alpha=@n42), B=@n44.RGB)) {name=Trail Color, nodeColor=<rgba 0.9922 1 0 1>}
  n45 = ComponentMask(None=@n29) {g=true}
  n46 = ScalarParameter() {group=General, parameterName=Trail End Cutoff, sliderMax=1}
  n47 = OneMinus(None=ComponentMask(None=@n29) {r=true})
  n48 = Multiply(A=Subtract(A=SmoothStep(Min=@n46, Value=@n47) {constMax=0.75, constMin=0.5}, B=Power(Base=@n47) {constExponent=10})) {constB=2}
  n49 = Panner(Coordinate=@n29) {speedX=-1}
  n50 = AppendVector(
   A=ScalarParameter() {defaultValue=1, parameterName=Tiling X, sliderMax=2}
   B=ScalarParameter() {defaultValue=1.5, group=General, parameterName=Tiling Y, sliderMax=10}
  )
  n51 = Multiply(
   A=TextureSample(UVs=Multiply(A=@n49, B=@n50)) {samplerType=SAMPLERTYPE_Grayscale, texture=/FortUILibrary/UIKit/Buttons/Textures/LowResBlurredNoise}.R
   B=ScalarParameter() {defaultValue=0.5, group=General, parameterName=Trail End Distortion Intensity, sliderMax=1}
  )
  n52 = Multiply(A=SmoothStep(Value=Sine(None=@n45) {period=2}) {constMax=0.15}, B=SmoothStep(Value=Subtract(A=@n48, B=@n51)) {constMin=0.25}) {constA=1}
  NamedRerouteDeclaration(None=Multiply(A=LinearInterpolate(A=@n27.A, B=@n28.A, Alpha=@n42), B=Multiply(A=@n44.A, B=@n52))) {name=Fade Opacity, nodeColor=<rgba 1 0 0.2109 1>}
 ]
}
